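# Physics asset for the CamelGram pickaxe mesh, a cosmetic two-handed melee harvesting tool.
# A single capsule body on the Melee_TwoHanded_Handle bone, slightly offset and rotated, gives a simple approximation of the weapon's shape.
# With only one body, it has no constraints, so the asset defines no articulated or jointed parts.
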PhysicsAsset /BRCosmetics/Weapons/FORT_Melee/Pickaxe_CamelGram/Meshes/CamelGram_Axe_PhysicsAsset
bodies {Melee_TwoHanded_Handle [{shapeType=Capsule, length=29.7884, radius=20.081, center=<2.6659 5.8141 -1.1285>, rotation=<rot -2.5859 -90 -0.385>}]}
constraints []
